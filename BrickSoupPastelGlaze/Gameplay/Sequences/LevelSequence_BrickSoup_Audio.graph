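# Level Sequence that times the audio and music event for BrickSoup PastelGlaze.
# It spawns an audio Actor with three sound sections and a music event Blueprint at the same fixed location, each on its own spawn window.
# The music Blueprint carries a long audio section, a keyed "Music Spawned" bool and an event repeater section.
# A root audio track adds an early sound with falling volume and keyed attach-actor references.
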LevelSequence /BrickSoupPastelGlaze/Gameplay/Sequences/LevelSequence_BrickSoup_Audio
bindings [
 BrickSoup_PastelGlaze_Audio {
  id=#1
  possessed_class=Actor
  tracks [
   {
    type=MovieScene3DTransformTrack
    sections [{
      type=MovieScene3DTransformSection
      channels [
       Location.X {type=MovieSceneScriptingDoubleChannel, default=5496, const [27052,1]}
       Location.Y {type=MovieSceneScriptingDoubleChannel, default=4048, const [-18076,1]}
       Location.Z {type=MovieSceneScriptingDoubleChannel, default=276, const [5940,1]}
       Rotation.X {type=MovieSceneScriptingDoubleChannel, default=0, const [0,1]}
       Rotation.Y {type=MovieSceneScriptingDoubleChannel, default=0, const [0,1]}
       Rotation.Z {type=MovieSceneScriptingDoubleChannel, default=0, const [0,1]}
       Scale.X {type=MovieSceneScriptingDoubleChannel, default=1, const [1,1]}
       Scale.Y {type=MovieSceneScriptingDoubleChannel, default=1, const [1,1]}
       Scale.Z {type=MovieSceneScriptingDoubleChannel, default=1, const [1,1]}
      ]
     }]
   }
   {
    type=MovieSceneSpawnTrack
    sections [{
      type=MovieSceneSpawnSection
      channels [None {type=MovieSceneScriptingBoolChannel, default=false, keys [[-1024,false],[-901,true],[9097,false]]}]
     }]
   }
   {
    type=MovieSceneAudioTrack
    display_name=Audio
    sections [
     {
      type=MovieSceneAudioSection
      end_frame=548
      start_frame=-850
      channels [
       Volume {type=MovieSceneScriptingFloatChannel, default=0.8}
       Pitch {type=MovieSceneScriptingFloatChannel, default=1}
      ]
     }
     {
      type=MovieSceneAudioSection
      end_frame=716
      start_frame=299
      channels [
       Volume {type=MovieSceneScriptingFloatChannel, default=1}
       Pitch {type=MovieSceneScriptingFloatChannel, default=1}
      ]
     }
     {
      type=MovieSceneAudioSection
      end_frame=6224
      start_frame=5729
      channels [
       Volume {type=MovieSceneScriptingFloatChannel, default=1}
       Pitch {type=MovieSceneScriptingFloatChannel, default=1}
      ]
     }
    ]
   }
  ]
 }
 B_BrickSoup_MusicEvent {
  id=#2
  possessed_class=B_BrickSoupPastelGlaze_MusicEvent_C
  tracks [
   {
    type=MovieSceneSpawnTrack
    sections [{
      type=MovieSceneSpawnSection
      channels [None {type=MovieSceneScriptingBoolChannel, default=false, keys [[-298,false],[-283,true],[6374,false]]}]
     }]
   }
   {
    type=MovieScene3DTransformTrack
    sections [{
      type=MovieScene3DTransformSection
      channels [
       Location.X {type=MovieSceneScriptingDoubleChannel, default=27052, const [27052,1]}
       Location.Y {type=MovieSceneScriptingDoubleChannel, default=-18076, const [-18076,1]}
       Location.Z {type=MovieSceneScriptingDoubleChannel, default=5940, const [5940,1]}
       Rotation.X {type=MovieSceneScriptingDoubleChannel, default=0, const [0,1]}
       Rotation.Y {type=MovieSceneScriptingDoubleChannel, default=0, const [0,1]}
       Rotation.Z {type=MovieSceneScriptingDoubleChannel, default=0, const [0,1]}
       Scale.X {type=MovieSceneScriptingDoubleChannel, default=1, const [1,1]}
       Scale.Y {type=MovieSceneScriptingDoubleChannel, default=1, const [1,1]}
       Scale.Z {type=MovieSceneScriptingDoubleChannel, default=1, const [1,1]}
      ]
     }]
   }
   {
    type=MovieSceneBoolTrack
    display_name=Music Spawned
    sections [{type=MovieSceneBoolSection, channels [None {type=MovieSceneScriptingBoolChannel, default=true, keys [[70,false],[363,true],[6224,false]]}]}]
   }
   {type=MovieSceneEventTrack, sections [{type=MovieSceneEventRepeaterSection, end_frame=9145, start_frame=310}]}
   {
    type=MovieSceneAudioTrack
    display_name=Audio
    sections [{
      type=MovieSceneAudioSection
      end_frame=6359
      start_frame=310
      channels [
       Volume {type=MovieSceneScriptingFloatChannel, default=1}
       Pitch {type=MovieSceneScriptingFloatChannel, default=1}
      ]
     }]
   }
  ]
 }
]
display_rate {denominator=1, numerator=60}
tick_resolution {denominator=1, numerator=24000}
tracks [{
  type=MovieSceneAudioTrack
  display_name=Audio
  sections [{
    type=MovieSceneAudioSection
    end_frame=618
    start_frame=163
    channels [
     Volume {type=MovieSceneScriptingFloatChannel, default=1.7, keys [[163,1.7],[618,1.3],[918,1]]}
     Pitch {type=MovieSceneScriptingFloatChannel, default=1, const [1,2]}
     AttachActor {
      type=MovieSceneScriptingActorReferenceChannel
      keys [[165, "<Struct 'MovieSceneObjectBindingID' (0x49dca735080) {}>"], [918, "<Struct 'MovieSceneObjectBindingID' (0x49dca734020) {}>"]]
     }
    ]
   }]
 }]
