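# Level Sequence for the Brick Soup buildup that times a block explosion effect alongside an event on the Day 3 spoon block prop.
# An event trigger track on the spoon prop fires a single keyed event.
# A spawned Niagara explosion actor follows an attach track for under half a second, and its component's system track runs the effect in a brief window within that span.
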LevelSequence /BrickSoupPastelGlaze/Gameplay/Sequences/LevelSequence_BrickSoup_Buildup
bindings [
 NiagaraComponent0 {
  id=#1
  parent_id=#2
  possessed_class=NiagaraComponent
  tracks [{type=MovieSceneNiagaraSystemTrack, sections [{type=MovieSceneNiagaraSystemSpawnSection, end_frame=5706, start_frame=4501}]}]
 }
 Prop_BrickSoupSpoon_Block_Day3 {
  id=#3
  possessed_class=Prop_BrickSoupSpoon_Block_Day3_C
  tracks [
   {
    type=MovieSceneEventTrack
    sections [{
      type=MovieSceneEventTriggerSection
      channels [None {type=MovieSceneScriptingEventChannel, const ["<Struct 'MovieSceneEvent' (0x49dca783ca0) {}>", 1]}]
     }]
   }
   {type=MovieSceneBindingLifetimeTrack, display_name=Binding Lifetime, sections [{type=MovieSceneBindingLifetimeSection}]}
  ]
 }
 "NS_BrickSoup_BlockExplosion (2)" {
  id=#2
  possessed_class=NiagaraActor
  tracks [
   {type=MovieSceneSpawnTrack, sections [{type=MovieSceneSpawnSection, channels [None {type=MovieSceneScriptingBoolChannel, default=true}]}]}
   {
    type=MovieScene3DTransformTrack
    sections [{
      type=MovieScene3DTransformSection
      channels [
       Location.X {type=MovieSceneScriptingDoubleChannel, default=0}
       Location.Y {type=MovieSceneScriptingDoubleChannel, default=0}
       Location.Z {type=MovieSceneScriptingDoubleChannel, default=0}
       Rotation.X {type=MovieSceneScriptingDoubleChannel, default=0}
       Rotation.Y {type=MovieSceneScriptingDoubleChannel, default=0}
       Rotation.Z {type=MovieSceneScriptingDoubleChannel, default=0}
       Scale.X {type=MovieSceneScriptingDoubleChannel, default=1}
       Scale.Y {type=MovieSceneScriptingDoubleChannel, default=1}
       Scale.Z {type=MovieSceneScriptingDoubleChannel, default=1}
      ]
     }]
   }
   {type=MovieScene3DAttachTrack, sections [{type=MovieScene3DAttachSection, end_frame=10569, start_frame=0}]}
  ]
 }
]
display_rate {denominator=1, numerator=60}
tick_resolution {denominator=1, numerator=24000}
tracks []
