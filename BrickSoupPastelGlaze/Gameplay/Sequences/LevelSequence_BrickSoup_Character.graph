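# Level Sequence that stages the timed appearance and disappearance of the PastelGlaze NPC mannequin in the BrickSoup event.
# Spawn and visibility keys bracket the scaled-up mannequin's overlapping skeletal animations, with reveal and disappear Niagara bursts timed around its showing and hiding.
# A fake-collision player-repulsor Blueprint shares the mannequin's spawn window.
# All actors carry attach tracks over zeroed local locations, alongside a BrickSoupSpoon_ActorPosition binding that has only a lifetime track.
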LevelSequence /BrickSoupPastelGlaze/Gameplay/Sequences/LevelSequence_BrickSoup_Character
bindings [
 NS_BrickSoup_NpcDisapear {
  id=#1
  possessed_class=NiagaraActor
  tracks [
   {type=MovieSceneSpawnTrack, sections [{type=MovieSceneSpawnSection, channels [None {type=MovieSceneScriptingBoolChannel, default=true}]}]}
   {
    type=MovieScene3DTransformTrack
    sections [{
      type=MovieScene3DTransformSection
      channels [
       Location.X {type=MovieSceneScriptingDoubleChannel, default=0}
       Location.Y {type=MovieSceneScriptingDoubleChannel, default=0}
       Location.Z {type=MovieSceneScriptingDoubleChannel, default=0}
       Rotation.X {type=MovieSceneScriptingDoubleChannel, default=0}
       Rotation.Y {type=MovieSceneScriptingDoubleChannel, default=0}
       Rotation.Z {type=MovieSceneScriptingDoubleChannel, default=0}
       Scale.X {type=MovieSceneScriptingDoubleChannel, default=1}
       Scale.Y {type=MovieSceneScriptingDoubleChannel, default=1}
       Scale.Z {type=MovieSceneScriptingDoubleChannel, default=1}
      ]
     }]
   }
   {type=MovieScene3DAttachTrack, sections [{type=MovieScene3DAttachSection, end_frame=10800, start_frame=0}]}
  ]
 }
 BrickSoupSpoon_ActorPosition {
  id=#2
  possessed_class=null
  tracks [{type=MovieSceneBindingLifetimeTrack, display_name=Binding Lifetime, sections [{type=MovieSceneBindingLifetimeSection}]}]
 }
 NiagaraComponent0 {
  id=#3
  parent_id=#1
  possessed_class=NiagaraComponent
  tracks [{type=MovieSceneNiagaraSystemTrack, sections [{type=MovieSceneNiagaraSystemSpawnSection, end_frame=6286, start_frame=5686}]}]
 }
 NiagaraComponent0 {
  id=#4
  parent_id=#5
  possessed_class=NiagaraComponent
  tracks [{type=MovieSceneNiagaraSystemTrack, sections [{type=MovieSceneNiagaraSystemSpawnSection, end_frame=900, start_frame=300}]}]
 }
 NS_BrickSoup_NpcReveal {
  id=#5
  possessed_class=NiagaraActor
  tracks [
   {type=MovieSceneSpawnTrack, sections [{type=MovieSceneSpawnSection, channels [None {type=MovieSceneScriptingBoolChannel, default=true}]}]}
   {
    type=MovieScene3DTransformTrack
    sections [{
      type=MovieScene3DTransformSection
      channels [
       Location.X {type=MovieSceneScriptingDoubleChannel, default=0}
       Location.Y {type=MovieSceneScriptingDoubleChannel, default=0}
       Location.Z {type=MovieSceneScriptingDoubleChannel, default=0}
       Rotation.X {type=MovieSceneScriptingDoubleChannel, default=0}
       Rotation.Y {type=MovieSceneScriptingDoubleChannel, default=0}
       Rotation.Z {type=MovieSceneScriptingDoubleChannel, default=0}
       Scale.X {type=MovieSceneScriptingDoubleChannel, default=1}
       Scale.Y {type=MovieSceneScriptingDoubleChannel, default=1}
       Scale.Z {type=MovieSceneScriptingDoubleChannel, default=1}
      ]
     }]
   }
   {type=MovieScene3DAttachTrack, sections [{type=MovieScene3DAttachSection, end_frame=10800, start_frame=0}]}
  ]
 }
 PastelGlaze {
  id=#6
  possessed_class=FortPlayerMannequin
  tracks [
   {
    type=MovieSceneSpawnTrack
    sections [{
      type=MovieSceneSpawnSection
      channels [None {type=MovieSceneScriptingBoolChannel, default=false, keys [[0,false],[135,true],[6438.03,false]]}]
     }]
   }
   {
    type=MovieScene3DTransformTrack
    sections [{
      type=MovieScene3DTransformSection
      channels [
       Location.X {type=MovieSceneScriptingDoubleChannel, default=0}
       Location.Y {type=MovieSceneScriptingDoubleChannel, default=0}
       Location.Z {type=MovieSceneScriptingDoubleChannel, default=0}
       Rotation.X {type=MovieSceneScriptingDoubleChannel, default=0}
       Rotation.Y {type=MovieSceneScriptingDoubleChannel, default=0}
       Rotation.Z {type=MovieSceneScriptingDoubleChannel, default=0}
       Scale.X {type=MovieSceneScriptingDoubleChannel, default=1.5}
       Scale.Y {type=MovieSceneScriptingDoubleChannel, default=1.5}
       Scale.Z {type=MovieSceneScriptingDoubleChannel, default=1.5}
      ]
     }]
   }
   {
    type=MovieSceneSkeletalAnimationTrack
    sections [
     {type=MovieSceneSkeletalAnimationSection, end_frame=460, start_frame=300}
     {type=MovieSceneSkeletalAnimationSection, end_frame=657, start_frame=460}
     {type=MovieSceneSkeletalAnimationSection, end_frame=4743, start_frame=583}
     {type=MovieSceneSkeletalAnimationSection, end_frame=5694, start_frame=4543}
     {type=MovieSceneSkeletalAnimationSection, end_frame=5820, start_frame=5575}
    ]
   }
   {type=MovieSceneGameplayCueTrack, display_name=Gameplay Cues, sections [{type=MovieSceneGameplayCueTriggerSection}]}
   {type=MovieScene3DAttachTrack, sections [{type=MovieScene3DAttachSection, end_frame=10779, start_frame=0}]}
   {
    type=MovieSceneVisibilityTrack
    sections [{
      type=MovieSceneVisibilitySection
      channels [None {type=MovieSceneScriptingBoolChannel, default=false, keys [[135,false],[274,true],[5820,false]]}]
     }]
   }
  ]
 }
 BP_BrickSoup_FakeCollisionVolume_and_PlayerRepulsor {
  id=#7
  possessed_class=BP_BrickSoup_FakeCollisionVolume_and_PlayerRepulsor_C
  tracks [
   {
    type=MovieSceneSpawnTrack
    sections [{
      type=MovieSceneSpawnSection
      channels [None {type=MovieSceneScriptingBoolChannel, default=false, keys [[0,false],[135,true],[6438.03,false]]}]
     }]
   }
   {
    type=MovieScene3DTransformTrack
    sections [{
      type=MovieScene3DTransformSection
      channels [
       Location.X {type=MovieSceneScriptingDoubleChannel, default=0}
       Location.Y {type=MovieSceneScriptingDoubleChannel, default=0}
       Location.Z {type=MovieSceneScriptingDoubleChannel, default=0}
       Rotation.X {type=MovieSceneScriptingDoubleChannel, default=0}
       Rotation.Y {type=MovieSceneScriptingDoubleChannel, default=0}
       Rotation.Z {type=MovieSceneScriptingDoubleChannel, default=0}
       Scale.X {type=MovieSceneScriptingDoubleChannel, default=1}
       Scale.Y {type=MovieSceneScriptingDoubleChannel, default=1}
       Scale.Z {type=MovieSceneScriptingDoubleChannel, default=1}
      ]
     }]
   }
   {type=MovieScene3DAttachTrack, sections [{type=MovieScene3DAttachSection, end_frame=10279, start_frame=0}]}
  ]
 }
]
display_rate {denominator=1, numerator=60}
tick_resolution {denominator=1, numerator=24000}
tracks []
