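# Level Sequence that stages the environment effects for the BrickSoup spoon block moment, with an event track on the Day3 spoon prop.
# A block explosion Niagara effect leads into a festive background effect and a warm, spawn-keyed spotlight, then an end burst closes the show.
# Explosion and burst actors use attach tracks, while the background and light have fixed transforms.
# The bound FX position actor carries an empty Gameplay Cue track.
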LevelSequence /BrickSoupPastelGlaze/Gameplay/Sequences/LevelSequence_BrickSoup_Environment
bindings [
 Prop_BrickSoupSpoon_Block_Day3 {
  id=#1
  possessed_class=Prop_BrickSoupSpoon_Block_Day3_C
  tracks [
   {
    type=MovieSceneEventTrack
    sections [{
      type=MovieSceneEventTriggerSection
      channels [None {type=MovieSceneScriptingEventChannel, const ["<Struct 'MovieSceneEvent' (0x49dca783520) {}>", 1]}]
     }]
   }
   {type=MovieSceneBindingLifetimeTrack, display_name=Binding Lifetime, sections [{type=MovieSceneBindingLifetimeSection}]}
  ]
 }
 BrickSoupSpotLight {
  id=#2
  possessed_class=SpotLight
  tracks [
   {
    type=MovieSceneSpawnTrack
    sections [{type=MovieSceneSpawnSection, channels [None {type=MovieSceneScriptingBoolChannel, default=false, keys [[0,false],[414,true],[5820,false]]}]}]
   }
   {
    type=MovieScene3DTransformTrack
    sections [{
      type=MovieScene3DTransformSection
      channels [
       Location.X {type=MovieSceneScriptingDoubleChannel, default=26439.6333}
       Location.Y {type=MovieSceneScriptingDoubleChannel, default=-17496}
       Location.Z {type=MovieSceneScriptingDoubleChannel, default=6909.9987}
       Rotation.X {type=MovieSceneScriptingDoubleChannel, default=-86.0721}
       Rotation.Y {type=MovieSceneScriptingDoubleChannel, default=-55.9501}
       Rotation.Z {type=MovieSceneScriptingDoubleChannel, default=-50.4805}
       Scale.X {type=MovieSceneScriptingDoubleChannel, default=1}
       Scale.Y {type=MovieSceneScriptingDoubleChannel, default=1}
       Scale.Z {type=MovieSceneScriptingDoubleChannel, default=1}
      ]
     }]
   }
  ]
 }
 NiagaraComponent0 {
  id=#3
  parent_id=#4
  possessed_class=NiagaraComponent
  tracks [{type=MovieSceneNiagaraSystemTrack, sections [{type=MovieSceneNiagaraSystemSpawnSection, end_frame=900, start_frame=300}]}]
 }
 NiagaraComponent0 {
  id=#5
  parent_id=#6
  possessed_class=NiagaraComponent
  tracks [{type=MovieSceneNiagaraSystemTrack, sections [{type=MovieSceneNiagaraSystemSpawnSection, end_frame=5849, start_frame=432}]}]
 }
 NiagaraComponent0 {
  id=#7
  parent_id=#8
  possessed_class=NiagaraComponent
  tracks [{type=MovieSceneNiagaraSystemTrack, sections [{type=MovieSceneNiagaraSystemSpawnSection, end_frame=6416, start_frame=5816}]}]
 }
 NS_BrickSoup_FestiveBackGround {
  id=#6
  possessed_class=NiagaraActor
  tracks [
   {type=MovieSceneSpawnTrack, sections [{type=MovieSceneSpawnSection, channels [None {type=MovieSceneScriptingBoolChannel, default=true}]}]}
   {
    type=MovieScene3DTransformTrack
    sections [{
      type=MovieScene3DTransformSection
      channels [
       Location.X {type=MovieSceneScriptingDoubleChannel, default=27022.3939}
       Location.Y {type=MovieSceneScriptingDoubleChannel, default=-18142.3628}
       Location.Z {type=MovieSceneScriptingDoubleChannel, default=5333.145}
       Rotation.X {type=MovieSceneScriptingDoubleChannel, default=0}
       Rotation.Y {type=MovieSceneScriptingDoubleChannel, default=0}
       Rotation.Z {type=MovieSceneScriptingDoubleChannel, default=0}
       Scale.X {type=MovieSceneScriptingDoubleChannel, default=1}
       Scale.Y {type=MovieSceneScriptingDoubleChannel, default=1}
       Scale.Z {type=MovieSceneScriptingDoubleChannel, default=1}
      ]
     }]
   }
  ]
 }
 LightComponent0 {
  id=#9
  parent_id=#2
  possessed_class=SpotLightComponent
  tracks [
   {
    type=MovieSceneFloatTrack
    display_name=Intensity
    sections [{type=MovieSceneFloatSection, channels [None {type=MovieSceneScriptingFloatChannel, default=30}]}]
   }
   {
    type=MovieSceneColorTrack
    display_name=Light Color
    sections [{
      type=MovieSceneColorSection
      channels [
       Color.R {type=MovieSceneScriptingFloatChannel, default=1}
       Color.G {type=MovieSceneScriptingFloatChannel, default=0.9301}
       Color.B {type=MovieSceneScriptingFloatChannel, default=0.7011}
       Color.A {type=MovieSceneScriptingFloatChannel, default=1}
      ]
     }]
   }
  ]
 }
 NS_BrickSoup_EndBurst {
  id=#8
  possessed_class=NiagaraActor
  tracks [
   {type=MovieSceneSpawnTrack, sections [{type=MovieSceneSpawnSection, channels [None {type=MovieSceneScriptingBoolChannel, default=true}]}]}
   {
    type=MovieScene3DTransformTrack
    sections [{
      type=MovieScene3DTransformSection
      channels [
       Location.X {type=MovieSceneScriptingDoubleChannel, default=0}
       Location.Y {type=MovieSceneScriptingDoubleChannel, default=0}
       Location.Z {type=MovieSceneScriptingDoubleChannel, default=0}
       Rotation.X {type=MovieSceneScriptingDoubleChannel, default=0}
       Rotation.Y {type=MovieSceneScriptingDoubleChannel, default=0}
       Rotation.Z {type=MovieSceneScriptingDoubleChannel, default=0}
       Scale.X {type=MovieSceneScriptingDoubleChannel, default=1}
       Scale.Y {type=MovieSceneScriptingDoubleChannel, default=1}
       Scale.Z {type=MovieSceneScriptingDoubleChannel, default=1}
      ]
     }]
   }
   {type=MovieScene3DAttachTrack, sections [{type=MovieScene3DAttachSection, end_frame=10569, start_frame=0}]}
  ]
 }
 BrickSoupSpoon_FXPosition {
  id=#10
  possessed_class=Actor
  tracks [
   {type=MovieSceneBindingLifetimeTrack, display_name=Binding Lifetime, sections [{type=MovieSceneBindingLifetimeSection}]}
   {type=MovieSceneGameplayCueTrack, display_name=Gameplay Cues, sections []}
  ]
 }
 "NS_BrickSoup_BlockExplosion (2)" {
  id=#4
  possessed_class=NiagaraActor
  tracks [
   {type=MovieSceneSpawnTrack, sections [{type=MovieSceneSpawnSection, channels [None {type=MovieSceneScriptingBoolChannel, default=true}]}]}
   {
    type=MovieScene3DTransformTrack
    sections [{
      type=MovieScene3DTransformSection
      channels [
       Location.X {type=MovieSceneScriptingDoubleChannel, default=0}
       Location.Y {type=MovieSceneScriptingDoubleChannel, default=0}
       Location.Z {type=MovieSceneScriptingDoubleChannel, default=0}
       Rotation.X {type=MovieSceneScriptingDoubleChannel, default=0}
       Rotation.Y {type=MovieSceneScriptingDoubleChannel, default=0}
       Rotation.Z {type=MovieSceneScriptingDoubleChannel, default=0}
       Scale.X {type=MovieSceneScriptingDoubleChannel, default=1}
       Scale.Y {type=MovieSceneScriptingDoubleChannel, default=1}
       Scale.Z {type=MovieSceneScriptingDoubleChannel, default=1}
      ]
     }]
   }
   {type=MovieScene3DAttachTrack, sections [{type=MovieScene3DAttachSection, end_frame=10569, start_frame=0}]}
  ]
 }
]
display_rate {denominator=1, numerator=60}
tick_resolution {denominator=1, numerator=24000}
tracks []
