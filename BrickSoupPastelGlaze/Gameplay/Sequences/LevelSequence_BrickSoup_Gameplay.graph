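# Gameplay Level Sequence for the BrickSoup content, which fires scripted events along a timeline with a 60 fps display rate.
# A root event track holds six trigger keys at irregular intervals, several in close pairs.
# The possessed Prop_BrickSoupSpoon_Block_Day3 actor has its own event trigger track and a binding lifetime track.
# No transform, animation or audio tracks are present.
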LevelSequence /BrickSoupPastelGlaze/Gameplay/Sequences/LevelSequence_BrickSoup_Gameplay
bindings [Prop_BrickSoupSpoon_Block_Day3 {
  id=#1
  possessed_class=Prop_BrickSoupSpoon_Block_Day3_C
  tracks [
   {
    type=MovieSceneEventTrack
    sections [{
      type=MovieSceneEventTriggerSection
      channels [None {type=MovieSceneScriptingEventChannel, const ["<Struct 'MovieSceneEvent' (0x49dca7847e0) {}>", 1]}]
     }]
   }
   {type=MovieSceneBindingLifetimeTrack, display_name=Binding Lifetime, sections [{type=MovieSceneBindingLifetimeSection}]}
  ]
 }]
display_rate {denominator=1, numerator=60}
tick_resolution {denominator=1, numerator=24000}
tracks [{
  type=MovieSceneEventTrack
  sections [{
    type=MovieSceneEventTriggerSection
    channels [None {
      type=MovieSceneScriptingEventChannel
      keys [
       [0, "<Struct 'MovieSceneEvent' (0x49dca780c80) {}>"]
       [29, "<Struct 'MovieSceneEvent' (0x49dca7810e0) {}>"]
       [5700, "<Struct 'MovieSceneEvent' (0x49dca781c20) {}>"]
       [6273, "<Struct 'MovieSceneEvent' (0x49dca781a40) {}>"]
       [11231, "<Struct 'MovieSceneEvent' (0x49dca781cc0) {}>"]
       [12102, "<Struct 'MovieSceneEvent' (0x49dca782120) {}>"]
      ]
     }]
   }]
 }]
